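# Sound cue for the third Pirate Island music stinger in the orchestral score.
# A single wave player feeds the root, playing the Pirate_Cove_Stinger_12 sound wave at half volume.
# It routes through the important in-game mission sound class and sets a very large max distance.
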SoundCue /Game/Sounds/Fort_Music/Orchestral_Score/Pirate_Island/Pirate_Island_Stinger_03_Cue
root=WavePlayer_0
nodes [{id=WavePlayer_0, properties {soundWaveAssetPtr=&/Pirate_Cove_Stinger_12}}]
properties {maxDistance=2097152, soundClassObject=&Fort_Mission_Ingame_Important_PSM, volumeMultiplier=0.5}
